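# Mono MetaSound source that plays weapon-fire shot sounds for the AlpacaLean glider cosmetic.
# Each change of the Fire integer input triggers a polyphonic wave player drawing on the SWC_AplaLean_Shots container; OnPlay is not wired.
# The source is looping and limited to one instance, so individual shots come from parameter changes inside it.
# Glider thrust attenuation applies, and a glider-open envelope follower bus modulates volume.
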MetaSoundSource /BRCosmetics/Sounds/Gliders/AlpacaLean_Gel/MSS_AlpacaLean_WeaponFire
inputs [
 Fire {type=Int32}
 UE.Source.OnPlay {type=Trigger}
]
interfaces [UE.OutputFormat.Mono:1.0, UE.Source:1.1]
outputs ["UE.OutputFormat.Mono.Audio:0" {type=Audio}]
pages [Default {
  input_defaults {Fire=15, UE.Source.OnPlay=null}
  nodes [
   TriggerOnValueChange_0 = UE.Trigger On Value Change:Int32(Value=^Fire)
   MetasoundPolyphonyInternalPolyphonyWavePlayer_0 = MetasoundPolyphonyInternal.PolyphonyWavePlayer(Play=@TriggerOnValueChange_0.Trigger) {also {defaults {Sound Wave Container=&/SWC_AplaLean_Shots}, version [0,1]}}
  ]
  links [{dst="@UE.OutputFormat.Mono.Audio:0", src=MetasoundPolyphonyInternalPolyphonyWavePlayer_0.Out Mono}]
 }]
properties {
 assetImportData=&self:AssetImportData
 attenuationSettings=&Fort_Player_Glider_Thrust_Athena_Att
 audioPropertiesSheet=&APS_Cosmetic_Glider_Loop
 bLooping=true
 duration=10000
 maxDistance=12500
 sampleRate=48000
 soundClassObject=&Fort_Gliders_Land
 volume=0.25
 concurrencyOverrides {maxCount=1}
 modulationSettings {
  volumeRouting=Union
  volumeModulationDestination {bEnableModulation=true, modulators [/Game/Sounds/Modulation/EnvelopeFollowers/Volume/CB_Glider_Open_EnvFollower]}
 }
}
