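# Looping stereo MetaSound source for the AlpacaLean_Gel glider's thrust audio, implementing the Fort.Glider interface.
# It wraps the shared MSS_Glider_Thrust_Loop graph, passing glider direction values, per-direction sounds, thresholds and pitch/volume modulation ranges through to stereo output.
# Only thrust and forward have loop arrays filled; forward, backward and sideways have start one-shots.
# Gunshot and standalone Front, Side and Back loop inputs are set but unconnected.
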MetaSoundSource /BRCosmetics/Sounds/Gliders/AlpacaLean_Gel/MSS_Glider_AlpacaLean_Gel_Thrust
inputs [
 Backward.Loop.Sounds {type=WaveAsset:Array}
 Backward.Modulator.PitchMax {type=Float}
 Backward.Modulator.PitchMin {type=Float}
 Backward.Modulator.VolumeMax {type=Float}
 Backward.Modulator.VolumeMin {type=Float}
 Backward.Sounds {type=WaveAsset:Array}
 Backward.Volume {type=Float}
 Backwards.OneshotThreashold {type=Float}
 "Backwards.ReverseOneshotThreashold " {type=Float}
 Backwards.ReverseOsTrigger {type=Bool}
 Forward.Loop.Sounds {type=WaveAsset:Array}
 Forward.Modulator.PitchMax {type=Float}
 Forward.Modulator.PitchMin {type=Float}
 Forward.Modulator.VolumeMax {type=Float}
 Forward.Modulator.VolumeMin {type=Float}
 Forward.OneshotThreshold {type=Float}
 Forward.ReverseOneShotThreshold {type=Float}
 Forward.Sounds {type=WaveAsset:Array}
 Forward.Volume {type=Float}
 Forwards.ReverseOsTrigger {type=Bool}
 Front.Loop {type=WaveAsset}
 Glider.AnyDirection {type=Float}
 Glider.Backwards {type=Float}
 Glider.Forwards {type=Float}
 Glider.Sideways {type=Float}
 GunShots {type=WaveAsset:Array}
 GunShotsGain {type=Float}
 Loop.Back {type=WaveAsset}
 Side.Loop {type=WaveAsset}
 Sideways.Loop.Sounds {type=WaveAsset:Array}
 Sideways.Modulator.PitchMax {type=Float}
 Sideways.Modulator.PitchMin {type=Float}
 Sideways.Modulator.VolumeMax {type=Float}
 Sideways.Modulator.VolumeMin {type=Float}
 Sideways.OneshotThreshold {type=Float}
 Sideways.ReverseOneShotThreshold {type=Float}
 Sideways.ReverseOsTrigger {type=Bool}
 Sideways.Sounds {type=WaveAsset:Array}
 Sideways.Volume {type=Float}
 Thrust.Loop.Sounds {type=WaveAsset:Array}
 Thrust.Modulator.OneShotThreshold {type=Float}
 Thrust.Modulator.PitchMax {type=Float}
 Thrust.Modulator.PitchMin {type=Float}
 Thrust.Modulator.VolumeMax {type=Float}
 Thrust.Modulator.VolumeMin {type=Float}
 Thrust.OneShotThreshold {type=Float}
 Thrust.OneShotVolume {type=Float}
 Thrust.ReverseOneShotThreshold {type=Float}
 Thrust.ReverseOsTrigger {type=Bool}
 Thrust.Sounds {type=WaveAsset:Array}
 Thrust.Volume {type=Float}
 UE.Source.OnPlay {type=Trigger}
 WeaponFireThreshold {type=Float}
]
interfaces [Fort.Glider:1.0, UE.OutputFormat.Stereo:1.0, UE.Source:1.1]
outputs [
 "UE.OutputFormat.Stereo.Audio:0" {type=Audio}
 "UE.OutputFormat.Stereo.Audio:1" {type=Audio}
]
pages [Default {
  input_defaults {
   Backward.Modulator.PitchMax=1
   Backward.Modulator.PitchMin=1
   Backward.Modulator.VolumeMax=1
   Backward.Modulator.VolumeMin=0
   Backward.Volume=0.75
   Backwards.OneshotThreashold=0.5
   "Backwards.ReverseOneshotThreashold "=0.2
   Backwards.ReverseOsTrigger=null
   Forward.Modulator.PitchMax=1.2
   Forward.Modulator.PitchMin=0.9
   Forward.Modulator.VolumeMax=0.15
   Forward.Modulator.VolumeMin=0
   Forward.OneshotThreshold=0.5
   Forward.ReverseOneShotThreshold=0.2
   Forward.Volume=0.75
   Forwards.ReverseOsTrigger=null
   Front.Loop=&/IP_Glider_AlpacaLean_Gel_ThrustLoop_Forward
   Glider.AnyDirection=0.0202
   Glider.Backwards=0
   Glider.Forwards=0
   Glider.Sideways=0.016
   GunShotsGain=0.8
   Loop.Back=&/IP_Glider_AlpacaLean_Gel_ThrustLoop_Backward
   Side.Loop=&/IP_Glider_AlpacaLean_Gel_ThrustLoop_Sideways
   Sideways.Modulator.PitchMax=1
   Sideways.Modulator.PitchMin=1
   Sideways.Modulator.VolumeMax=1
   Sideways.Modulator.VolumeMin=0
   Sideways.OneshotThreshold=0.5
   Sideways.ReverseOneShotThreshold=0.2
   Sideways.ReverseOsTrigger=null
   Sideways.Volume=0.75
   Thrust.Modulator.OneShotThreshold=0.5
   Thrust.Modulator.PitchMax=1.1
   Thrust.Modulator.PitchMin=1
   Thrust.Modulator.VolumeMax=0.5
   Thrust.Modulator.VolumeMin=0.25
   Thrust.OneShotThreshold=0.5
   Thrust.OneShotVolume=1
   Thrust.ReverseOneShotThreshold=0.2
   Thrust.ReverseOsTrigger=null
   Thrust.Volume=1
   UE.Source.OnPlay=null
   WeaponFireThreshold=0.5
   Backward.Loop.Sounds []
   Backward.Sounds [
    &/IP_Glider_AlpacaLean_Gel_ThurstStart_Small_01
    &/IP_Glider_AlpacaLean_Gel_ThurstStart_Small_02
    &/IP_Glider_AlpacaLean_Gel_ThurstStart_Small_03
   ]
   Forward.Loop.Sounds [&IP_Glider_AlpacaLean_Gel_ThrustLoop_Forward]
   Forward.Sounds [
    &/IP_Glider_AlpacaLean_Gel_ThurstStart_01
    &/IP_Glider_AlpacaLean_Gel_ThurstStart_02
    &/IP_Glider_AlpacaLean_Gel_ThurstStart_03
   ]
   GunShots [
    &/IP_Glider_AlpacaLean_Gel_ThrustLoop_Shot_01
    &/IP_Glider_AlpacaLean_Gel_ThrustLoop_Shot_02
    &/IP_Glider_AlpacaLean_Gel_ThrustLoop_Shot_03
    &/IP_Glider_AlpacaLean_Gel_ThrustLoop_Shot_04
    &/IP_Glider_AlpacaLean_Gel_ThrustLoop_Shot_05
    &/IP_Glider_AlpacaLean_Gel_ThrustLoop_Shot_06
   ]
   Sideways.Loop.Sounds []
   Sideways.Sounds [
    &IP_Glider_AlpacaLean_Gel_ThurstStart_Small_03
    &IP_Glider_AlpacaLean_Gel_ThurstStart_Small_02
    &IP_Glider_AlpacaLean_Gel_ThurstStart_Small_01
   ]
   Thrust.Loop.Sounds [&/IP_Glider_AlpacaLean_Gel_ThrustLoop]
   Thrust.Sounds []
  }
  nodes [
   InitVariable.Audio()
   InitVariable.Audio()
   InitVariable.Time()
   InitVariable.Trigger()
   InitVariable.Trigger()
   InitVariable.Trigger()
   InitVariable.Trigger()
   InitVariable.Trigger()
   InitVariable.Trigger()
   InitVariable.Trigger()
   InitVariable.Trigger()
   InitVariable.Trigger()
   InitVariable.Trigger()
   MSS_Glider_Thrust_Loop_0 = /FortMetasound/Cosmetics/Glider/MSS_Glider_Thrust_Loop(
    Backward.Loop.Sounds=^Backward.Loop.Sounds
    Backward.Modulator.PitchMax=^Backward.Modulator.PitchMax
    Backward.Modulator.PitchMin=^Backward.Modulator.PitchMin
    Backward.Modulator.VolumeMax=^Backward.Modulator.VolumeMax
    Backward.Modulator.VolumeMin=^Backward.Modulator.VolumeMin
    Backward.Sounds=^Backward.Sounds
    Backward.Volume=^Backward.Volume
    Backwards.OneshotThreashold=^Backwards.OneshotThreashold
    "Backwards.ReverseOneshotThreashold "=^"Backwards.ReverseOneshotThreashold "
    Backwards.ReverseOsTrigger=^Backwards.ReverseOsTrigger
    Forward.Loop.Sounds=^Forward.Loop.Sounds
    Forward.Modulator.PitchMax=^Forward.Modulator.PitchMax
    Forward.Modulator.PitchMin=^Forward.Modulator.PitchMin
    Forward.Modulator.VolumeMax=^Forward.Modulator.VolumeMax
    Forward.Modulator.VolumeMin=^Forward.Modulator.VolumeMin
    Forward.OneshotThreshold=^Forward.OneshotThreshold
    Forward.ReverseOneShotThreshold=^Forward.ReverseOneShotThreshold
    Forward.Sounds=^Forward.Sounds
    Forward.Volume=^Forward.Volume
    Forwards.ReverseOsTrigger=^Forwards.ReverseOsTrigger
    Glider.AnyDirection=^Glider.AnyDirection
    Glider.Backwards=^Glider.Backwards
    Glider.Forwards=^Glider.Forwards
    Glider.Sideways=^Glider.Sideways
    Sideways.Loop.Sounds=^Sideways.Loop.Sounds
    Sideways.Modulator.PitchMax=^Sideways.Modulator.PitchMax
    Sideways.Modulator.PitchMin=^Sideways.Modulator.PitchMin
    Sideways.Modulator.VolumeMax=^Sideways.Modulator.VolumeMax
    Sideways.Modulator.VolumeMin=^Sideways.Modulator.VolumeMin
    Sideways.OneshotThreshold=^Sideways.OneshotThreshold
    Sideways.ReverseOneShotThreshold=^Sideways.ReverseOneShotThreshold
    Sideways.ReverseOsTrigger=^Sideways.ReverseOsTrigger
    Sideways.Sounds=^Sideways.Sounds
    Sideways.Volume=^Sideways.Volume
    Thrust.Loop.Sounds=^Thrust.Loop.Sounds
    Thrust.Modulator.OneShotThreshold=^Thrust.Modulator.OneShotThreshold
    Thrust.Modulator.PitchMax=^Thrust.Modulator.PitchMax
    Thrust.Modulator.PitchMin=^Thrust.Modulator.PitchMin
    Thrust.Modulator.VolumeMax=^Thrust.Modulator.VolumeMax
    Thrust.Modulator.VolumeMin=^Thrust.Modulator.VolumeMin
    Thrust.OneShotThreshold=^Thrust.OneShotThreshold
    Thrust.OneShotVolume=^Thrust.OneShotVolume
    Thrust.ReverseOneShotThreshold=^Thrust.ReverseOneShotThreshold
    Thrust.ReverseOsTrigger=^Thrust.ReverseOsTrigger
    Thrust.Sounds=^Thrust.Sounds
    Thrust.Volume=^Thrust.Volume
    UE.Source.OnPlay=^UE.Source.OnPlay
   )
  ]
  links [
   {dst="@UE.OutputFormat.Stereo.Audio:0", src=MSS_Glider_Thrust_Loop_0.UE.OutputFormat.Stereo.Audio:0}
   {dst="@UE.OutputFormat.Stereo.Audio:1", src=MSS_Glider_Thrust_Loop_0.UE.OutputFormat.Stereo.Audio:1}
  ]
 }]
properties {
 assetImportData=&self:AssetImportData
 attenuationSettings=&Fort_Player_Glider_Thrust_Athena_Att
 audioPropertiesSheet=&APS_Cosmetic_Glider_Loop
 bLooping=true
 duration=10000
 maxDistance=12500
 numChannels=2
 outputFormat=Stereo
 sampleRate=48000
 soundClassObject=&Fort_Gliders_Land
 volume=0.75
 concurrencyOverrides {bLimitToOwner=true, maxCount=1}
 concurrencySet [&Athena_Glider_Thrust_Conc]
 modulationSettings {
  volumeRouting=Union
  volumeModulationDestination {bEnableModulation=true, modulators [/Game/Sounds/Modulation/EnvelopeFollowers/Volume/CB_Glider_Open_EnvFollower]}
 }
}
